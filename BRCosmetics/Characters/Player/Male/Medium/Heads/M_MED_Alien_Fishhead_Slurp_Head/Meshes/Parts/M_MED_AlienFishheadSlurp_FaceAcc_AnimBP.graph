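# Animation Blueprint for the face accessory part of the male Alien Fishhead Slurp player head.
# The AnimGraph copies pose and curves from the attached parent mesh.
# A GetShieldPercentage function returns the player pawn's shield over max shield, with the divisor floored at one.
# It falls back to the FrontEnd_Idle_Bypass curve when the FortPlayerAnimInstance cast fails, or zero without a valid mesh; nothing here calls it.
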AnimBlueprint /BRCosmetics/Characters/Player/Male/Medium/Heads/M_MED_Alien_Fishhead_Slurp_Head/Meshes/Parts/M_MED_AlienFishheadSlurp_FaceAcc_AnimBP
target_skeleton=&/M_MED_AlienFishheadSlurp_FaceAcc_Skeleton
graphs [
 AnimGraph {nodes [Output Pose(Result=Copy Pose From Mesh() {bCopyCurves=true, bUseAttachedParent=true}) {layerGroup=None}]}
 EventGraph {links [], nodes []}
 GetShieldPercentage {nodes [
   GetShieldPercentage() {is=title K2Node_FunctionEntry}
   n1 = GetOwnerSkeletalMeshComponent()
   > n2 = Is Valid(InputObject=@n1)
   Return Node!(execute=@n2.Is Not Valid, ShieldValue:double=0)
   n3 = GetAnimInstance(self=@n1)
   > n4 = Cast To FortPlayerAnimInstance(Object=@n3)
   n5 = Get FortPlayerPawn!(self:Fort Player Anim Instance Object Reference=@n4) -> (FortPlayerPawn:Fort Player Pawn Object Reference)
   n6 = GetMaxShield!(self:Fort Pawn Object Reference=@n5) -> (ReturnValue:float)
   > Return Node(ShieldValue=float / float(A=GetShield!(self:Fort Pawn Object Reference=@n5) -> (ReturnValue:float), B="Max (Float)"(A=@n6, B=1)) {is=title K2Node_CallFunction})
   Return Node(execute=@n4.CastFailed, ShieldValue=GetCurveValue(self=@n3, CurveName=FrontEnd_Idle_Bypass))
  ]}
]
variables []
